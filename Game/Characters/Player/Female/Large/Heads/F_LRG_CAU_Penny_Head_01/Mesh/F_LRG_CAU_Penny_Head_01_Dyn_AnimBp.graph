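# Post-style animation Blueprint for the Penny large female head, built for dynamic hair motion but currently passing Input Pose straight to Output Pose.
# An unconnected chain of Anim Dynamics nodes simulates eight dyn_hair bones, with angular limits and a spherical limit on neck_01.
# Pawn 2D speed, remapped and clamped, drives their shared alpha.
# A RigidBody node also sits unconnected, and the event graph holds no logic.
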AnimBlueprint /Game/Characters/Player/Female/Large/Heads/F_LRG_CAU_Penny_Head_01/Mesh/F_LRG_CAU_Penny_Head_01_Dyn_AnimBp
target_skeleton=/Game/Characters/Player/Common/Fortnite_Base_Head/Fortnite_Base_Head_Export_Skeleton
graphs [
 AnimGraph {nodes [
   n1 = float * float(A=1, B=MapRangeClamped(Value=Get PawnSpeed2D(), InRangeA=0, InRangeB=100, OutRangeA=.2))
   n2 = Anim Dynamics - Bone: dyn_hair_04(ComponentPose=Local To Component(), Alpha=@n1, ExternalForce="0,0,0") {
    bGravityOverrideInSimSpace=true
    bUseSphericalLimits=true
    numSolverIterationsPostUpdate=4
    numSolverIterationsPreUpdate=16
    alphaBoolBlend {blendOption=Linear}
    boundBone {boneName=dyn_hair_04}
    physicsBodyDefinitions [{
      sphereCollisionRadius=0
      boundBone {boneName=dyn_hair_04}
      boxExtents=<20 20 10>
      constraintSetup {twistAxis=AxisZ, angularLimitsMax=<20 20 0>, angularLimitsMin=<-20 -20 0>, angularTarget=<0 0 0>}
      localJointOffset=<0 0 -5>
     }]
    sphericalLimits [{limitRadius=6, drivingBone {boneName=neck_01}, sphereLocalOffset=<0 0 0>}]
   }
   n3 = Anim Dynamics - Bone: dyn_hair_08(ComponentPose=@n2, Alpha=@n1, ExternalForce="0,0,0") {
    bGravityOverrideInSimSpace=true
    bUseSphericalLimits=true
    numSolverIterationsPostUpdate=4
    numSolverIterationsPreUpdate=16
    alphaBoolBlend {blendOption=Linear}
    boundBone {boneName=dyn_hair_08}
    physicsBodyDefinitions [{
      sphereCollisionRadius=0
      boundBone {boneName=dyn_hair_08}
      boxExtents=<20 20 10>
      constraintSetup {twistAxis=AxisZ, angularLimitsMax=<20 20 0>, angularLimitsMin=<-20 -20 0>, angularTarget=<0 0 0>}
      localJointOffset=<0 0 -5>
     }]
    sphericalLimits [{limitRadius=6, drivingBone {boneName=neck_01}, sphereLocalOffset=<0 0 0>}]
   }
   n4 = Anim Dynamics - Bone: dyn_hair_01(ComponentPose=@n3, Alpha=@n1) {
    bGravityOverrideInSimSpace=true
    bUseSphericalLimits=true
    numSolverIterationsPostUpdate=4
    numSolverIterationsPreUpdate=16
    alphaBoolBlend {blendOption=Linear}
    boundBone {boneName=dyn_hair_01}
    physicsBodyDefinitions [{
      sphereCollisionRadius=0
      boundBone {boneName=dyn_hair_01}
      boxExtents=<20 20 10>
      constraintSetup {twistAxis=AxisZ, angularLimitsMax=<20 20 0>, angularLimitsMin=<-20 -20 0>, angularTarget=<0 0 0>}
      localJointOffset=<0 0 -5>
     }]
    sphericalLimits [{limitRadius=6, drivingBone {boneName=neck_01}, sphereLocalOffset=<0 0 0>}]
   }
   n5 = Anim Dynamics - Bone: dyn_hair_07(ComponentPose=@n4, Alpha=@n1, ExternalForce="0,0,0") {
    bGravityOverrideInSimSpace=true
    bUseSphericalLimits=true
    numSolverIterationsPostUpdate=4
    numSolverIterationsPreUpdate=16
    alphaBoolBlend {blendOption=Linear}
    boundBone {boneName=dyn_hair_07}
    physicsBodyDefinitions [{
      sphereCollisionRadius=0
      boundBone {boneName=dyn_hair_07}
      boxExtents=<20 20 10>
      constraintSetup {twistAxis=AxisZ, angularLimitsMax=<20 20 0>, angularLimitsMin=<-20 -20 0>, angularTarget=<0 0 0>}
      localJointOffset=<0 0 -5>
     }]
    sphericalLimits [{limitRadius=6, drivingBone {boneName=neck_01}, sphereLocalOffset=<0 0 0>}]
   }
   n6 = Anim Dynamics - Bone: dyn_hair_02(ComponentPose=@n5, Alpha=@n1) {
    bGravityOverrideInSimSpace=true
    bUseSphericalLimits=true
    numSolverIterationsPostUpdate=4
    numSolverIterationsPreUpdate=16
    alphaBoolBlend {blendOption=Linear}
    boundBone {boneName=dyn_hair_02}
    physicsBodyDefinitions [{
      sphereCollisionRadius=0
      boundBone {boneName=dyn_hair_02}
      boxExtents=<20 20 10>
      constraintSetup {twistAxis=AxisZ, angularLimitsMax=<20 20 0>, angularLimitsMin=<-20 -20 0>, angularTarget=<0 0 0>}
      localJointOffset=<0 0 -5>
     }]
    sphericalLimits [{limitRadius=6, drivingBone {boneName=neck_01}, sphereLocalOffset=<0 0 0>}]
   }
   n7 = Anim Dynamics - Bone: dyn_hair_05(ComponentPose=@n6, Alpha=@n1, ExternalForce="0,0,0") {
    bGravityOverrideInSimSpace=true
    bUseSphericalLimits=true
    numSolverIterationsPostUpdate=4
    numSolverIterationsPreUpdate=16
    alphaBoolBlend {blendOption=Linear}
    boundBone {boneName=dyn_hair_05}
    physicsBodyDefinitions [{
      sphereCollisionRadius=0
      boundBone {boneName=dyn_hair_05}
      boxExtents=<20 20 10>
      constraintSetup {twistAxis=AxisZ, angularLimitsMax=<20 20 0>, angularLimitsMin=<-20 -20 0>, angularTarget=<0 0 0>}
      localJointOffset=<0 0 -5>
     }]
    sphericalLimits [{limitRadius=6, drivingBone {boneName=neck_01}, sphereLocalOffset=<0 0 0>}]
   }
   n8 = Anim Dynamics - Bone: dyn_hair_03(ComponentPose=@n7, Alpha=@n1, ExternalForce="0,0,0") {
    bGravityOverrideInSimSpace=true
    bUseSphericalLimits=true
    numSolverIterationsPostUpdate=4
    numSolverIterationsPreUpdate=16
    alphaBoolBlend {blendOption=Linear}
    boundBone {boneName=dyn_hair_03}
    physicsBodyDefinitions [{
      sphereCollisionRadius=0
      boundBone {boneName=dyn_hair_03}
      boxExtents=<20 20 10>
      constraintSetup {twistAxis=AxisZ, angularLimitsMax=<20 20 0>, angularLimitsMin=<-20 -20 0>, angularTarget=<0 0 0>}
      localJointOffset=<0 0 -2>
     }]
    sphericalLimits [{limitRadius=6, drivingBone {boneName=neck_01}, sphereLocalOffset=<0 0 0>}]
   }
   n9 = Anim Dynamics - Bone: dyn_hair_06(ComponentPose=@n8, Alpha=@n1, ExternalForce="0,0,0") {
    bGravityOverrideInSimSpace=true
    bUseSphericalLimits=true
    numSolverIterationsPostUpdate=4
    numSolverIterationsPreUpdate=16
    alphaBoolBlend {blendOption=Linear}
    boundBone {boneName=dyn_hair_06}
    physicsBodyDefinitions [{
      sphereCollisionRadius=0
      boundBone {boneName=dyn_hair_06}
      boxExtents=<20 20 10>
      constraintSetup {twistAxis=AxisZ, angularLimitsMax=<20 20 0>, angularLimitsMin=<-20 -20 0>, angularTarget=<0 0 0>}
      localJointOffset=<0 0 -1>
     }]
    sphericalLimits [{limitRadius=6, drivingBone {boneName=neck_01}, sphereLocalOffset=<0 0 0>}]
   }
   Component To Local(ComponentPose=@n9)
   RigidBody() {alphaBoolBlend {blendOption=Linear}}
   AnimGraphNode_SubInput_0 = Input Pose()
   Output Pose(Result=@AnimGraphNode_SubInput_0) {layerGroup=None}
  ]}
 EventGraph {nodes [
   TryGetPawnOwner()
   Event BlueprintUpdateAnimation()
  ]}
]
variables []
